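# Post-pose AnimBP for the F_MED_Space_Cuddles cosmetic body that corrects shoulder pad and thigh helper bones.
# AnimGraph copies the attached parent's pose, then chains additive Transform Bone nodes.
# EventGraph sets their alphas: shoulder pads shift in the frontend, while skydive floating, with parachute open, and when aiming pistol-class weapons (interpolated).
# Thigh helper bones scale by each thigh's swing angle from reference pose, remapped through clamped ranges.
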AnimBlueprint /BRCosmetics/Characters/Player/Female/Medium/Bodies/F_MED_Space_Cuddles/Meshes/F_MED_Space_Cuddles_AnimBP
target_skeleton=&/F_MED_Space_Cuddles_Skeleton
graphs [
 AnimGraph {nodes [
   n1 = Get IsFrontend() -> (IsFrontend:double)
   n2 = "Transform (Modify) Bone - Bone: dyn_shoulderpad_r"(
    ComponentPose=Local To Component(LocalPose=Copy Pose From Mesh() {bUseAttachedParent=true})
    Alpha=@n1
    Translation="0,1,0"
    Rotation="10,10,-30.000"
    Scale="1,1,1"
   ) {
    rotationMode=BMM_Additive
    rotationSpace=BCS_BoneSpace
    translationMode=BMM_Additive
    translationSpace=BCS_BoneSpace
    alphaBoolBlend {blendOption=Linear}
    boneToModify {boneName=dyn_shoulderpad_r}
    rotation=<rot 10 10 -30>
    translation {y=1}
   }
   n3 = "Transform (Modify) Bone - Bone: dyn_shoulderpad_l"(ComponentPose=@n2, Alpha=@n1, Translation="1.25,-0.75,0", Rotation="0.000,0.000,-20", Scale="1,1,1") {
    rotationMode=BMM_Additive
    rotationSpace=BCS_BoneSpace
    translationMode=BMM_Additive
    translationSpace=BCS_BoneSpace
    alphaBoolBlend {blendOption=Linear}
    boneToModify {boneName=dyn_shoulderpad_l}
    rotation {roll=-20}
    translation=<1.25 -0.75>
   }
   n4 = "Transform (Modify) Bone - Bone: dyn_sub_thigh_r"(ComponentPose=@n3, Alpha=Get legSwingDelta_r() -> (legSwingDelta_r:double), Translation="0,0,0", Rotation="0,0,0", Scale="4,4,4") {
    scaleMode=BMM_Additive
    scaleSpace=BCS_BoneSpace
    alphaBoolBlend {blendOption=Linear}
    boneToModify {boneName=dyn_sub_thigh_r}
    scale=<4 4 4>
   }
   n5 = "Transform (Modify) Bone - Bone: dyn_sub_thigh_l"(ComponentPose=@n4, Alpha=Get legSwingDelta_l() -> (legSwingDelta_l:double), Translation="0,0,0", Rotation="0,0,0", Scale="4,4,4") {
    scaleMode=BMM_Additive
    scaleSpace=BCS_BoneSpace
    alphaBoolBlend {blendOption=Linear}
    boneToModify {boneName=dyn_sub_thigh_l}
    scale=<4 4 4>
   }
   n6 = "To Float (Boolean)"(InBool=Get Alt_weapon() -> (Alt_weapon:bool))
   n7 = "Transform (Modify) Bone - Bone: dyn_shoulderpad_r"(ComponentPose=@n5, Alpha=@n6, Translation="0,0,-0.5", Rotation="-10,0,10", Scale="1,1,1") {
    rotationMode=BMM_Additive
    rotationSpace=BCS_BoneSpace
    translationMode=BMM_Additive
    translationSpace=BCS_BoneSpace
    alphaBoolBlend {blendOption=Linear}
    alphaScaleBiasClamp {bInterpResult=true, interpSpeedDecreasing=20, interpSpeedIncreasing=20}
    boneToModify {boneName=dyn_shoulderpad_r}
    rotation {pitch=-10, roll=10}
    translation {z=-0.5}
   }
   n8 = "Transform (Modify) Bone - Bone: dyn_shoulderpad_l"(ComponentPose=@n7, Alpha=@n6, Translation="2,0,0.000", Rotation="-10,-20,20", Scale="1,1,1") {
    rotationMode=BMM_Additive
    rotationSpace=BCS_BoneSpace
    translationMode=BMM_Additive
    translationSpace=BCS_BoneSpace
    alphaBoolBlend {blendOption=Linear}
    alphaScaleBiasClamp {bInterpResult=true, interpSpeedDecreasing=20, interpSpeedIncreasing=20}
    boneToModify {boneName=dyn_shoulderpad_l}
    rotation=<rot -10 -20 20>
    translation {x=2}
   }
   n9 = Get SkydiveParachute_Alpha() -> (SkydiveParachute_Alpha:double)
   n10 = "Transform (Modify) Bone - Bone: dyn_shoulderpad_r"(ComponentPose=@n8, Alpha=@n9, Translation="4.000,-0.000,0", Rotation="0,30,-40", Scale="1,1,1") {
    rotationMode=BMM_Additive
    rotationSpace=BCS_BoneSpace
    translationMode=BMM_Additive
    translationSpace=BCS_BoneSpace
    alphaBoolBlend {blendOption=Linear}
    boneToModify {boneName=dyn_shoulderpad_r}
    rotation {roll=-40, yaw=30}
    translation {x=4}
   }
   n11 = "Transform (Modify) Bone - Bone: dyn_shoulderpad_l"(ComponentPose=@n10, Alpha=@n9, Translation="4,0,0", Rotation="0,-30,40", Scale="1,1,1") {
    rotationMode=BMM_Additive
    rotationSpace=BCS_BoneSpace
    translationMode=BMM_Additive
    translationSpace=BCS_BoneSpace
    alphaBoolBlend {blendOption=Linear}
    boneToModify {boneName=dyn_shoulderpad_l}
    rotation {roll=40, yaw=-30}
    translation {x=4}
   }
   n12 = Get SkydiveFloat_Alpha() -> (SkydiveFloat_Alpha:double)
   n13 = "Transform (Modify) Bone - Bone: dyn_shoulderpad_r"(ComponentPose=@n11, Alpha=@n12, Translation="2,-0.000,-1.5", Rotation="0,0,0", Scale="1,1,1") {
    rotationMode=BMM_Additive
    rotationSpace=BCS_BoneSpace
    translationMode=BMM_Additive
    translationSpace=BCS_BoneSpace
    alphaBoolBlend {blendOption=Linear}
    boneToModify {boneName=dyn_shoulderpad_r}
    translation {x=2, z=-1.5}
   }
   n14 = "Transform (Modify) Bone - Bone: dyn_shoulderpad_l"(ComponentPose=@n13, Alpha=@n12, Translation="2,0,-1.5", Rotation="0,0,0", Scale="1,1,1") {
    rotationMode=BMM_Additive
    rotationSpace=BCS_BoneSpace
    translationMode=BMM_Additive
    translationSpace=BCS_BoneSpace
    alphaBoolBlend {blendOption=Linear}
    boneToModify {boneName=dyn_shoulderpad_l}
    translation {x=2, z=-1.5}
   }
   Output Pose(Result=Component To Local(ComponentPose=@n14)) {layerGroup=None}
  ]}
 EventGraph {nodes [
   Event BlueprintUpdateAnimation()
   > n1 = Sequence!(execute) -> (then_0, then_1, then_2:Exec, then_3:Exec, then_4:Exec)
   > Set SkydiveFloat_Alpha(SkydiveFloat_Alpha:double=SelectFloat(A=1, B=0, bPickA=Get bIsSkydiveFloating())) -> (Output_Get:double)
   Set SkydiveParachute_Alpha(execute=@n1.then_1, SkydiveParachute_Alpha:double=SelectFloat(A=1, B=0, bPickA=Get bIsParachuteOpened())) -> (Output_Get:double)
   n2 = Cast To FortPlayerAnimInstance(execute=@n1.then_2, Object=GetAnimInstance(self=GetOwnerSkeletalMeshComponent()))
   > Set Alt_weapon(Alt_weapon:bool=false) -> (Output_Get:bool)
   > n3 = Switch on EFortWeaponCoreAnimation!(execute, Selection:EFortWeaponCoreAnimation Enum=Get WeaponCoreAnim(self=@n2)) -> (
    Melee:Exec
    Pistol:Exec
    Shotgun:Exec
    PaperBlueprint:Exec
    Rifle:Exec
    MeleeOneHand:Exec
    MachinePistol:Exec
    RocketLauncher:Exec
    GrenadeLauncher:Exec
    GoingCommando:Exec
    AssaultRifle:Exec
    TacticalShotgun:Exec
    SniperRifle:Exec
    TrapPlacement:Exec
    ShoulderLauncher:Exec
    AbilityDecoTool:Exec
    Crossbow:Exec
    C4:Exec
    RemoteControl:Exec
    DualWield:Exec
    AR_BullPup:Exec
    AR_ForwardGrip:Exec
    MedPackPaddles:Exec
    SMG_P90:Exec
    AR_DrumGun:Exec
    Consumable_Small:Exec
    Consumable_Large:Exec
    Balloon:Exec
    MountedTurret:Exec
    CreativeTool:Exec
    ExplosiveBow:Exec
    AshtonIndigo:Exec
    AshtonChicago:Exec
    MeleeDualWield:Exec
    Unarmed:Exec
   )
   Branch(execute=@n3.Pistol, Condition=Get bIsTargeting())
   > Set Alt_weapon(Alt_weapon=true)
   n4 = GetOwnerSkeletalMeshComponent()
   Set legSwingDelta_r(
    execute=@n1.then_3
    legSwingDelta_r:double=MapRangeClamped(Value=GetTwistAndSwingAngleOfDeltaRotationFromRefPose(self=@n4, BoneName=thigh_r).OutSwingAngle, InRangeA=150, InRangeB=75)
   ) -> (Output_Get:double)
   > Set legSwingDelta_l(legSwingDelta_l:double=MapRangeClamped(Value=GetTwistAndSwingAngleOfDeltaRotationFromRefPose(self=@n4, BoneName=thigh_l).OutSwingAngle, InRangeA=43, InRangeB=110)) -> (Output_Get:double)
   Set IsFrontend(execute=@n1.then_4, IsFrontend:double=SelectFloat(A=1, B=0, bPickA=Get bIsInFrontEnd())) -> (Output_Get:double)
  ]}
]
variables [
 Alt_weapon {type=boolean}
 IsFrontend {type=number}
 legSwingDelta_l {type=number}
 legSwingDelta_r {type=number}
 SkydiveFloat_Alpha {type=number}
 SkydiveParachute_Alpha {type=number}
]
